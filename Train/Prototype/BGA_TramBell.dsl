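; Prototype tram bell actor that rings when a Fort pawn interacts with it through the "Ring Bell" prompt or when it takes a hit.
; Both the interaction and the OnDamaged delegate bound at BeginPlay call PlayBell, which spawns the ring sound attached to the root.
; An OnCooldown flag, cleared by a one-second timer, blocks repeat rings.
; Incoming damage is modified to zero, and the tick and overlap events are empty.
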(event EventBeginPlay
  (EventDispatchers|BindEventtoOnDamaged (delegate Custom|Bell_OnDamaged)))

(event Custom|Bell_OnDamaged
  (params (DamagedActor :type Actor%20Object%20Reference) (Damage :type Float%20%28single-precision%29) (InstigatedBy :type Controller%20Object%20Reference) (DamageCauser :type Actor%20Object%20Reference) (HitLocation :type Vector) (FHitComponent :type Primitive%20Component%20Object%20Reference) (BoneName :type Name) (Momentum :type Vector))
  (CallFunction|PlayBell))

(event Collision|EventActorBeginOverlap
  (params (OtherActor :type Actor%20Object%20Reference)))

(event EventTick
  (params (DeltaSeconds :type Float%20%28single-precision%29)))

(event Custom|PlayBell
  (Variables|Default|GetOnCooldown
    (:then)
    (:else
      (Audio|SpawnSoundAttached "/Train/Prototype/Trambell_Ring_Rundown.Trambell_Ring_Rundown" (Variables|Default|GetDefaultSceneRoot))
      (Variables|Default|SetOnCooldown true)
      (Utilities|Time|SetTimerbyEvent (delegate Custom|ClearCooldown) 1.0))))

(event Custom|ClearCooldown
  (Variables|Default|SetOnCooldown false))

(event Interaction|EventBlueprintOnInteract
  (params (InteractingPawn :type Fort%20Pawn%20Object%20Reference) (InteractionBeingAttempted :type EInteractionBeingAttempted%20Enum :default "FirstInteraction"))
  (CallFunction|PlayBell))

(fn BlueprintCanInteract
  (params (InteractingPawn :type Fort%20Pawn%20Object%20Reference) (InteractionBeingAttempted :type EInteractionBeingAttempted%20Enum :default "FirstInteraction") (InteractionType :type TInteractionType%20Enum :default "IT_NoInteraction"))
  (return true))

(fn BlueprintGetInteractionString
  (params (InteractingPawn :type Fort%20Pawn%20Object%20Reference) (InteractionBeingAttempted :type EInteractionBeingAttempted%20Enum :default "FirstInteraction"))
  (return "NSLOCTEXT(\"[1BB73D8C87EC5BB0DC107784E7F6EF31]\", \"56F2495342DC3E4DCAF960925C1AEEF9\", \"Ring Bell\")"))

(fn BlueprintModifyIncomingDamage
  (params (Damage :type Float%20%28single-precision%29) (DamageTags :type Gameplay%20Tag%20Container%20Structure%20%28by%20ref%29) (EffectContext :type Gameplay%20Effect%20Context%20Handle%20Structure%20%28by%20ref%29) (EventInstigator :type Controller%20Object%20Reference) (DamageCauser :type Actor%20Object%20Reference))
  (return 0.0))

(fn ConstructionScript)
