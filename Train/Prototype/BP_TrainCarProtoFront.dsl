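; Prototype front train car Blueprint that blows a periodic whistle and knocks away actors targeted for impulse.
; A looping server timer increments a RepNotify whistle counter, forcing client notification despite train dormancy so the whistle audio plays.
; Targeted pawns get an immediate vertical launch, covering slow GAS knockback activation, then gameplay effects and a cue; non-pawns are woken as vehicles.
; Remote instances also enable onboard audio processing for the local player.
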(event EventBeginPlay
  (|CallParentFunction)
  (Networking|SwitchHasAuthority
    (:Authority
      (Utilities|Time|SetTimerbyEvent (EventDispatchers|CreateEvent :!event Custom|Event_NotifyBlowWhistle) 2e+01 true))
    (:Remote
      (CallFunction|EnableAudioProcessingforLocalPlayerOnboard 1.0 (Variables|Default|GetFortSoundIndicator) (Variables|Default|GetAudioShapeLineList)))))

(event EventOnActorTargetedForImpulse
  (params (Actor :type Actor%20Object%20Reference))
  (CallFunction|ApplyFrontImpulse Actor))

(event Custom|Event_NotifyBlowWhistle
  ;[0 0 0] Todo @Seth: drive dynamically on station arrival as well
  (Variables|Default|SetWhistleCount (+ (Variables|Default|GetWhistleCount) 1)) ; Since the train is always dormant, we modify a variable with RepNotify to force a notification to the client
)

(event Custom|Event_BlowWhistle
  ;[0 0 0] Todo @Seth: drive dynamically on station arrival as well
  (Audio|Components|Audio|Play (Variables|Default|GetAudioWhistle) 0.0))

(fn ApplyFrontImpulse
  (params (Actor :type Actor%20Object%20Reference))
  (bind _asfort_pawn (Utilities|Casting|CastToFortPawn Actor)
    (:then
      (Character|LaunchCharacter _asfort_pawn (Math|Vector|MakeVector 0.0 0.0 (ScalableFloat|GetValueatLevel (Variables|Default|GetLaunchSpeed))) true true) ; We need this initial launch since the knockback GE/GA takes too long to activate, and our character could already be on the train. GE/GA still handles knockback status so it handles cancelling other abilities
      (Fortnite|ApplyGameplayEffectsToActor _asfort_pawn (Variables|Default|GetEffects%20to%20Apply) (Transformation|GetActorLocation _asfort_pawn) false 0 (Ability|GetFortAbilitySystemComponent self))
      (GameplayCue|ExecuteGameplayCue _asfort_pawn (Variables|Default|GetPlayerLaunchGameplayCueTag)))
    (:CastFailed
      (bind _asfort_vehicle_interface (Utilities|Casting|CastToFortVehicleInterface Actor))
      (Vehicle|WakeUp _asfort_vehicle_interface))))

(fn OnRep_WhistleCount
  (CallFunction|EventBlowWhistle))

(fn ConstructionScript
  (|CallParentFunction))
