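# M_Craft_Proto is a minimal Default Lit prototype material, apparently for craft meshes in the Train prototype.
# It exposes a Color vector parameter, defaulting to light grey, and a fairly rough Roughness scalar parameter for tuning through instances.
# It is enabled for Nanite, with automatic usage flag setting in the editor turned off.
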Material /Train/Prototype/M_Craft_Proto
graph {
 properties {bAutomaticallySetUsageInEditor=false, bUsedWithNanite=true, usedShadingModels=Default Lit}
 nodes [
  ScalarParameter() {defaultValue=0.8, parameterName=Roughness}
  VectorParameter() {parameterName=Color, defaultValue=<rgba 0.8646 0.8646 0.8646 0>}
 ]
}
