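# Prototype train sound cue for a tram bell ring rundown, built from buoy bell and doorbell samples on the ambient props sound class.
# A mixer layers a buoy bell and a doorbell immediately, then two more buoy bell hits on staggered fixed delays.
# A large cue-wide pitch multiplier raises all layers far above their source pitch.
# Each layer has its own attenuation node, with low-pass occlusion on the buoy bells only.
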SoundCue /Train/Prototype/Trambell_Ring_Rundown
root=Mixer_0
nodes [
 {
  id=Attenuation_0
  children [WavePlayer_2]
  properties {bOverrideAttenuation=true, attenuationOverrides {bEnableOcclusion=true, occlusionInterpolationTime=3, occlusionLowPassFilterFrequency=800}}
 }
 {
  id=Attenuation_1
  children [WavePlayer_3]
  properties {bOverrideAttenuation=true, attenuationOverrides {bEnableOcclusion=true, occlusionInterpolationTime=3, occlusionLowPassFilterFrequency=800}}
 }
 {
  id=Attenuation_2
  children [WavePlayer_0]
  properties {bOverrideAttenuation=true, attenuationOverrides {bEnableOcclusion=true, occlusionInterpolationTime=3, occlusionLowPassFilterFrequency=800}}
 }
 {id=Attenuation_3, children [WavePlayer_1], properties {bOverrideAttenuation=true}}
 {id=Delay_0, children [Attenuation_0], properties {delayMax=0.25, delayMin=0.25}}
 {id=Delay_1, children [Attenuation_1], properties {delayMax=0.5, delayMin=0.5}}
 {id=Mixer_0, children [Attenuation_2, Attenuation_3, Delay_0, Delay_1], properties {inputVolume [1,1,1,1]}}
 {id=WavePlayer_0, properties {soundWaveAssetPtr=/Game/Sounds/Helios/Environment_Actors/Buoy/Spot_Buoy_Bell_4}}
 {id=WavePlayer_1, properties {soundWaveAssetPtr=/Game/Athena/Sounds/Quests/Doorbells/Doorbell_1_Out}}
 {id=WavePlayer_2, properties {soundWaveAssetPtr=&Spot_Buoy_Bell_4}}
 {id=WavePlayer_3, properties {soundWaveAssetPtr=/Game/Sounds/Helios/Environment_Actors/Buoy/Spot_Buoy_Bell_5}}
]
properties {
 maxDistance=4000
 pitchMultiplier=6
 soundClassObject=&Fort_Ambient_Props_Athena
 volumeMultiplier=0.928
 attenuationOverrides {distanceAlgorithm=Logarithmic, stereoSpread=512}
 concurrencyOverrides {bLimitToOwner=true, maxCount=2, resolutionRule=StopOldest}
}
